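# Bubble material for the Overcharge vehicle boost effect, an unlit masked material built for Niagara sprites, ribbons and mesh particles.
# Bubble texture color passes through MF_FX_EmissiveAdjustments for emissive, while its alpha is multiplied by a procedural dissolve mask.
# The mask samples cloud noise in radial coordinates and adds a soft circular center gradient.
# Inverted particle alpha offsets that sum before a saturated sine, so particle alpha animates the breakup.
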Material /VehicleCosmetics/Boosts/Boost_Overcharge/Material/Bubble_MAT
graph {
 properties {
  bAutomaticallySetUsageInEditor=false
  bUsedWithNiagaraMeshParticles=true
  bUsedWithNiagaraRibbons=true
  bUsedWithNiagaraSprites=true
  blendMode=BLEND_Masked
  shadingModel=MSM_Unlit
  usedShadingModels=Unlit
 }
 nodes [
  Constant3Vector() {constant=<rgba 0.7396 0.3203 0.1713 1>}
  Constant3Vector() {constant=<rgba 0.875 0.2526 0.0911 1>}
  n1 = TextureSample() {texture=/VehicleCosmetics/Boosts/Boost_Overcharge/Texture/T_VCVFX_Boost_Overcharge_Bubble}
  MaterialFunctionCall(Emissive=Multiply(A=@n1.RGB)) {materialFunction=&MF_FX_EmissiveAdjustments}
  n2 = Subtract(A=TextureCoordinate() {uTiling=2, vTiling=2}, B=Constant() {r=1})
  n3 = StaticBool() {value=true}
  n4 = TextureSample(UVs=Multiply(
    A=Constant2Vector() {g=0.1, r=2}
    B=MaterialFunctionCall(Vector or UVs=@n2, Swizzle Coordinate Output=@n3) {materialFunction=&VectorToRadialValue}.Radial Coordinates
   )) {texture=/VehicleCosmetics/FX/Textures/Noise/T_VCVFX_Clouds_A}
  n5 = Add(A=TextureCoordinate(), B=Constant() {r=-0.5})
  n6 = OneMinus(None=DotProduct(A=@n5, B=@n5))
  n7 = Constant() {r=12}
  MaterialExpressionConstantClamp_6 = Saturate(None=Multiply(A=Power(Base=@n6, Exp=@n7), B=Constant() {r=0.5}))
  MaterialExpressionConstantClamp_5 = Saturate(None=Add(A=ConstantBiasScale(None=@n4.R) {bias=0.5}, B=@MaterialExpressionConstantClamp_6))
  MaterialExpressionVertexColor_2 = ParticleColor()
  n8 = OneMinus(None=Add(A=@MaterialExpressionVertexColor_2.A, B=Constant() {r=0.2}))
  MaterialExpressionConstantClamp_4 = Saturate(None=Sine(None=Add(A=@MaterialExpressionConstantClamp_5, B=@n8)) {period=2})
  Multiply(A=Multiply(A=@n1.A, B=@MaterialExpressionConstantClamp_4))
 ]
}
