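# Additive, unlit, two-sided Niagara material for flickering electric arcs in the Overcharge boost effect.
# Zero-tiled panners turn cloud noise into sharply thresholded flicker signals, switching between red and blue channels of a UV1 lightning texture.
# A third signal, its speed scaled by particle blue, blinks the arc on and off.
# The mask times Color and particle alpha feeds MF_FX_EmissiveAdjustments; the red-to-red lerp has no effect.
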Material /VehicleCosmetics/Boosts/Boost_Overcharge/Material/ElectricArc_MAT
graph {
 properties {
  bAutomaticallySetUsageInEditor=false
  bUsedWithNiagaraMeshParticles=true
  bUsedWithNiagaraRibbons=true
  bUsedWithNiagaraSprites=true
  blendMode=BLEND_Additive
  shadingModel=MSM_Unlit
  twoSided=true
  usedShadingModels=Unlit
 }
 nodes [
  n1 = ComponentMask(None=TextureCoordinate() {uTiling=1.01}) {r=true}
  MaterialExpressionConstantClamp_11 = Saturate(None=Multiply(A=OneMinus(None=@n1), B=Constant() {r=1}))
  n2 = VectorParameter() {parameterName=Color, defaultValue=<rgba 0.1 0.25 2 6>}
  MaterialExpressionVertexColor_4 = ParticleColor()
  n3 = TextureSample(UVs=TextureCoordinate() {coordinateIndex=1}) {texture=/VehicleCosmetics/Boosts/Boost_Overcharge/Texture/T_VCVFX_Boost_Overcharge_OverCharge_Lightning}
  n4 = TextureCoordinate() {uTiling=0, vTiling=0}
  n5 = AppendVector(
   A=TextureSample(UVs=Panner(Coordinate=@n4) {speedX=1, speedY=0.65}) {texture=/VehicleCosmetics/FX/Textures/Noise/T_VCVFX_Clouds_A}.R
   B=TextureSample(UVs=Panner(Coordinate=@n4) {speedX=-1, speedY=-0.3}) {texture=&T_VCVFX_Clouds_A}.R
  )
  n6 = Constant() {r=0.2}
  MaterialExpressionConstantClamp_14 = Saturate(None=Multiply(A=Subtract(A=@n5, B=@n6), B=Constant() {r=35}))
  n7 = LinearInterpolate(A=@n3.R, B=@n3.R, Alpha=ComponentMask(None=@MaterialExpressionConstantClamp_14) {r=true})
  n8 = ComponentMask(None=@MaterialExpressionConstantClamp_14) {g=true}
  n9 = Time() {bIgnorePause=true}
  n10 = TextureSample(UVs=Panner(Coordinate=TextureCoordinate() {uTiling=0, vTiling=0}, Time=Multiply(A=@MaterialExpressionVertexColor_4.B, B=@n9)) {speedX=2, speedY=3}) {texture=&T_VCVFX_Clouds_A}
  n11 = Constant() {r=0.4}
  MaterialExpressionConstantClamp_13 = Saturate(None=Multiply(A=Subtract(A=@n10.R, B=@n11), B=Constant() {r=55}))
  MaterialExpressionConstantClamp_12 = Saturate(None=Multiply(A=LinearInterpolate(A=@n7, B=@n3.B, Alpha=@n8), B=@MaterialExpressionConstantClamp_13))
  MaterialFunctionCall(Emissive=Multiply(A=Multiply(A=@n2.RGB, B=@MaterialExpressionVertexColor_4.A), B=@MaterialExpressionConstantClamp_12)) {materialFunction=&MF_FX_EmissiveAdjustments}
 ]
}
