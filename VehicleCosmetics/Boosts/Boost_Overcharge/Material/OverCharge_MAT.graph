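# Translucent, unlit, two-sided material for Niagara sprite, ribbon and mesh particles in the Overcharge vehicle boost cosmetic, using splash textures by default.
# A splash normal map is transformed and dotted with the Z axis to fake directional shading, tinting particle color before MF_FX_EmissiveAdjustments.
# A separate branch erodes the splash texture by subtracting inverted particle color, then multiplies and saturates it into a sharpened mask.
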Material /VehicleCosmetics/Boosts/Boost_Overcharge/Material/OverCharge_MAT
graph {
 properties {
  bAutomaticallySetUsageInEditor=false
  bUsedWithNiagaraMeshParticles=true
  bUsedWithNiagaraRibbons=true
  bUsedWithNiagaraSprites=true
  blendMode=BLEND_Translucent
  shadingModel=MSM_Unlit
  twoSided=true
  usedShadingModels=Unlit
 }
 nodes [
  n1 = TextureSampleParameter2D() {parameterName=Texture, texture=/DealershipSharedLibrary/FX/Textures/Gradient_Masks/T_VCVFX_Splash01_A}
  MaterialExpressionVertexColor_0 = ParticleColor()
  n2 = OneMinus(None=@MaterialExpressionVertexColor_0.RGB)
  MaterialExpressionConstantClamp_0 = Saturate(None=Multiply(A=Subtract(A=@n1.RGB, B=@n2), B=Constant() {r=3}))
  n3 = Transform(None=TextureSampleParameter2D() {parameterName=Normal, samplerType=SAMPLERTYPE_Normal, texture=/DealershipSharedLibrary/FX/Textures/Normal/T_VCVFX_Splash01_N}.RGB)
  n4 = DotProduct(A=@n3, B=Constant3Vector() {constant=<rgba 0 0 1 0>})
  MaterialFunctionCall(Emissive=Multiply(A=@MaterialExpressionVertexColor_0.RGB, B=ConstantBiasScale(None=@n4) {bias=1.2})) {materialFunction=&MF_FX_EmissiveAdjustments}
  Reroute(None=@n3)
 ]
}
