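# Unlit, masked, two-sided master material for Niagara sprite, ribbon and mesh particles in the Overcharge vehicle boost effect.
# A static switch uses either two normal textures blended by particle red, or a single Normal parameter texture, carrying alpha alongside.
# That normal drives faked lighting: tinted fresnel rim, biased diffuse against a fixed Z vector, and thresholded half-vector specular.
# The three terms use squared color parameters and sum into MF_FX_EmissiveAdjustments.
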Material /VehicleCosmetics/Boosts/Boost_Overcharge/Material/Overcharge_Master_New_MAT
graph {
 properties {
  bAutomaticallySetUsageInEditor=false
  bUsedWithNiagaraMeshParticles=true
  bUsedWithNiagaraRibbons=true
  bUsedWithNiagaraSprites=true
  blendMode=BLEND_Masked
  refractionMethod=RM_2DOffset
  shadingModel=MSM_Unlit
  twoSided=true
  usedShadingModels=Unlit
 }
 nodes [
  n1 = CameraVectorWS()
  n2 = TextureSample() {texture=/VehicleCosmetics/Boosts/Boost_Overcharge/Texture/T_VCVFX_Boost_Overcharge_Liquid_01_N}
  n3 = TextureSample() {texture=/VehicleCosmetics/Boosts/Boost_Overcharge/Texture/T_VCVFX_Boost_Overcharge_TEST_03}
  MaterialExpressionVertexColor_1 = ParticleColor()
  n4 = TextureSampleParameter2D() {parameterName=Normal, texture=/VehicleCosmetics/Boosts/Boost_Overcharge/Texture/T_VCVFX_Boost_Overcharge_TEST_01}
  n5 = StaticSwitchParameter(True=LinearInterpolate(A=@n2.RGB, B=@n3.RGB, Alpha=@MaterialExpressionVertexColor_1.R), False=@n4.RGB) {defaultValue=true, parameterName=Blend}
  n6 = OneMinus(None=DotProduct(A=Transform(None=@n1) {transformSourceType=TRANSFORMSOURCE_World, transformType=TRANSFORM_Tangent}, B=@n5))
  n7 = Constant() {r=0.3}
  MaterialExpressionConstantClamp_1 = Saturate(None=Multiply(A=Subtract(A=@n6, B=@n7), B=ScalarParameter() {defaultValue=1, parameterName=Fresnel_Sharpness}))
  n8 = VectorParameter() {parameterName=FresnelColor, defaultValue=<rgba 1 0.1 0 2>}
  n9 = Multiply(A=@MaterialExpressionConstantClamp_1, B=Multiply(A=@n8.RGB, B=@n8.RGB))
  n10 = VectorParameter() {parameterName=Color, defaultValue=<rgba 1 0.2 0 2>}
  n11 = Transform(None=@n5) {transformSourceType=TRANSFORMSOURCE_World, transformType=TRANSFORM_Tangent}
  n12 = Constant3Vector() {constant=<rgba 0 0 1 0>}
  MaterialExpressionConstantClamp_2 = Saturate(None=DotProduct(A=@n11, B=@n12))
  n13 = ScalarParameter() {defaultValue=0.35, parameterName=Bias}
  n14 = Multiply(A=Multiply(A=@n10.RGB, B=@n10.RGB), B=Multiply(A=Add(A=@MaterialExpressionConstantClamp_2, B=@n13), B=Constant() {r=0.5}))
  n15 = VectorParameter() {parameterName=SpecColor, defaultValue=<rgba 1 0.5 0.1 3>}
  n16 = Multiply(A=@n15.RGB, B=@n15.RGB)
  n17 = CameraVectorWS()
  n18 = Normalize(VectorInput=Add(A=Transform(None=@n12), B=Transform(None=@n17) {transformSourceType=TRANSFORMSOURCE_World, transformType=TRANSFORM_Tangent}))
  MaterialExpressionConstantClamp_3 = Saturate(None=Subtract(A=DotProduct(A=@n11, B=@n18), B=Constant() {r=0.9}))
  MaterialFunctionCall(Emissive=Add(A=Add(A=@n9, B=@n14), B=Multiply(A=@n16, B=@MaterialExpressionConstantClamp_3))) {materialFunction=&MF_FX_EmissiveAdjustments}
  Reroute(None=StaticSwitchParameter(True=LinearInterpolate(A=@n2.A, B=@n3.A, Alpha=@MaterialExpressionVertexColor_1.R), False=@n4.A) {defaultValue=true, parameterName=Blend})
  Reroute(None=@n5)
 ]
}
